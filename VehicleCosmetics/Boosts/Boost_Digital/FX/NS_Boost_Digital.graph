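# Niagara system for the Digital vehicle boost cosmetic, trailing ribbons and small glowing sprites from the car's boost ports.
# Both emitters use filtered socket counts, spawn at skeletal mesh locations, fire backward along negative X and inherit source movement by curve.
# The ribbon has quality-dependent spawn rates and picks fresnel color and diffuse contrast from arrays indexed by color variant ID.
# Painted-variant overrides color both emitters, while jittered sprites change brightness and material parameters when painted.
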NiagaraSystem /VehicleCosmetics/Boosts/Boost_Digital/FX/NS_Boost_Digital
dependencies {
 usedDataInterfaces [
  &NiagaraDataInterfaceCurve
  &NiagaraDataInterfaceArrayColor
  &NiagaraDataInterfaceArrayFloat
  &NiagaraDataInterfaceSkeletalMesh
 ]
 usedDynamicInputs [&Multiply_LinearColorByFloat]
 usedModules [
  &self:SystemSpawnScript.SystemScriptSource.SystemScriptGraph.NiagaraNodeAssignment_0.SetVariables_#1
  &SystemState
 ]
}
emitters {
 NE_Boost_Ribbon {
  emitterSpawnScript [/SetVariables(
    Emitter.DiffuseContrast:NiagaraFloat=&SelectFloatFromArray(
     Direct Array Index=$System.ColorVariantID
     Float Selection Array=DataInterface() {propertyValues {FloatData [1,1,1,1.35,1,2,1.25,1,1,0.5,1,1,1,1.25]}}
    )
    Emitter.FresnelColor:LinearColor=&SelectLinearColorFromArray(
     Array Sampling Mode=::NewEnumerator1
     Direct Array Index=$System.ColorVariantID
     Color Selection Array=DataInterface() {propertyValues {ColorData [
        <rgba 1.6668 0 100 1>
        <rgba 1 1 1 1>
        <rgba 0.25 0.25 0.25 1>
        <rgba 8 1.6 0 1>
        <rgba 0 1.6667 100 1>
        <rgba 15 0.0625 0 1>
        <rgba 0 10 0 1>
        <rgba 10 10 10 1>
        <rgba 10 2.5 0 1>
        <rgba 10 2.5 0 1>
        <rgba 50 5 49.9999 1>
        <rgba 10 0 20 1>
        <rgba 0 0.7103 1 1>
        <rgba 0 20 30 1>
       ]}}
     Random Seed=?
    )
   ) {isSetParameters=true}]
  emitterUpdateScript [
   &EmitterState(Life Cycle Mode=::NewEnumerator1, Loop Duration=1, Recalculate Duration Each Loop=true, UseLoopDelay=false)
   &VC_FilteredSocketCount()
   &VC_RateQualityHandler(High Spawn Rate=60, Low Spawn Rate=30, Use Custom Rate?=true)
   &VC_SpawnRibbon_v1(Number of Ribbons=$Emitter.PortCount, SpawnRate=$Emitter.Spawn Rate)
  ]
  particleSpawnScript [
   &VC_PaintedVariantsOverride(
    Black=<rgba 2 2 2 1.5>
    BurntSienna=<rgba 5 4.1667 0 4>
    Cobalt=<rgba 1.25 2.0625 50 1>
    Crimson=<rgba 1 0.054 0.05 3>
    DefaultColor=<rgba 0.5 0 30 1>
    ForestGreen=<rgba 0 4 0 2>
    Grey=<rgba 6 6 6 1>
    Lime=<rgba 1.3333 2 0 1.5>
    Orange=<rgba 4 1.6667 0 5>
    Pink=<rgba 15 4.5 12.375 1>
    Purple=<rgba 0.43 0.05 1 3>
    Saffron=<rgba 2.25 1.7644 0.1687 1.15>
    SkyBlue=<rgba 0 3.3333 5 2>
    TitaniumWhite=<rgba 20 20 20 1>
   )
   &VC_InitializeRibbon_v1(
    Color=$Particles.Color
    Custom Offset=true
    Enable Skeletal Position=true
    Facing Axis:ENiagaraOrientationAxis=::YAxis
    Lifetime=1
    Local Vector Offset=&AddVectorToPosition(
     Position=/Niagara/DynamicInputs/TypeConversions/MakeCustomPositionFromBool(Bool:NiagaraBool=$User.bIsFrontendPreview, True Position:NiagaraPosition=<0 0 0>, False Position:NiagaraPosition=<0 0 0>)
     Vector=$User.DTCarOffset
    )
    Offset Amount=1
    Ribbon Width=40
   )
   &AddVelocity(Velocity=<-2000 0 0>) {module=AddVelocity001}
  ]
  particleUpdateScript [
   &ParticleState()
   &InheritSourceMovement(
    Applied Position Scale=&Multiply_VectorByFloat(
     Vector=<1 1 1>
     Float=&FloatFromCurve(FloatCurve=DataInterface() {propertyValues {Curve {keys [
          {interpMode=RCIM_Cubic, leaveTangent=-0.0051, tangentMode=RCTM_Break, time=0, value=1}
          {arriveTangent=-0.247, interpMode=RCIM_Cubic, tangentMode=RCTM_Break, time=1, value=0.9}
         ]}}})
    )
    ApplyForce=false
    ApplyPosition=true
    ApplyPositionFalloff=false
   )
   &SolveForcesAndVelocity(Clamp Velocity=false, Limit Acceleration=false)
   &VC_UpdateRibbon_v1(Enable Skeletal Position=true)
   &ScaleColor(Scale RGB=&VectorFromFloat(Value=1.35), ScaleA=true, ScaleRGB=true)
   &DynamicMaterialParameters(Write Parameter Index 0=::NewEnumerator0)
   &ScaleRibbonWidth(Ribbon Width Scale=&FloatFromCurve(FloatCurve=DataInterface() {propertyValues {Curve {keys [<at 0 1.75>,<at 0.2 2>,<at 1 1>]}}}))
  ]
  rendererClasses [&NiagaraRibbonRendererProperties]
  renderers [{rendererClass=&NiagaraRibbonRendererProperties}]
 }
 NE_VC_Boost_Sprite_Parent {
  emitterUpdateScript [
   &EmitterState()
   &VC_FilteredSocketCount()
   &SpawnRate(SpawnRate=100, Use Spawn Probability=false)
  ]
  particleSpawnScript [
   &VC_PaintedVariantsOverride(
    Black=<rgba 0.03 0.03 0.03 1>
    BurntSienna=<rgba 0.5 0.0625 0 1>
    Cobalt=<rgba 0 0 4 1>
    Crimson=<rgba 0.15 0.0006 0 1>
    DefaultColor=<rgba 0.0083 0 0.5 1>
    ForestGreen=<rgba 0 0.275 0 1>
    Pink=<rgba 2 0 1.3333 1>
    Purple=<rgba 0.1667 0 0.25 1>
    SkyBlue=<rgba 0 0.2917 0.5 1>
   )
   &InitializeParticle(
    Color=$Particles.Color
    Lifetime Max=0.75
    Lifetime Min=0.5
    Lifetime Mode=::NewEnumerator1
    Sprite Rotation Mode=::NewEnumerator2
    Sprite Size Mode=::NewEnumerator1
    Uniform Sprite Size Max=4
    Uniform Sprite Size Min=2
    UsePositionOffset=false
    Write Material Random=true
   )
   &PartitionParticles(Number Of Partitions=$Emitter.PortCount)
   &SkeletalMeshLocation(
    Bone / Socket Index=$Particles.PartitionParticles.Partition
    Bone / Socket Sampling Mode=::NewEnumerator5
    MaskSpawnGroup=false
    Mesh Sampling Type=::NewEnumerator6
    Skeletal Mesh=$System.SkeletalMesh
   )
   &AddVelocity(Velocity=&RandomRangeVector(Minimum=<-1500 -5 -5>, Maximum=<-1500 5 5>))
   &ShapeLocation(MaskSpawnGroup=false, Override Seed=false, Sphere Radius=30, UseOwnerScale=true)
  ]
  particleUpdateScript [
   &ParticleState()
   &InheritSourceMovement(
    Applied Position Scale=&Multiply_VectorByFloat(
     Vector=<1 1 1>
     Float=&FloatFromCurve(FloatCurve=DataInterface() {propertyValues {Curve {keys [
          {
           arriveTangentWeight=0.4252
           interpMode=RCIM_Cubic
           leaveTangentWeight=0.4252
           tangentMode=RCTM_User
           tangentWeightMode=RCTWM_WeightedBoth
           time=0.05
           value=1
          }
          {interpMode=RCIM_Cubic, tangentMode=RCTM_User, time=1, value=0}
         ]}}})
    )
    ApplyForce=false
    ApplyPosition=true
    ApplyPositionFalloff=false
   )
   &SpriteRotationRate(Rotation Rate=&RandomRangeFloat(Minimum=-2, Maximum=2))
   &ScaleSpriteSize(Uniform Curve Sprite Scale=DataInterface() {propertyValues {Curve {keys [
        <at 0 1 RCIM_Cubic>
        {arriveTangent=0.0641, interpMode=RCIM_Cubic, leaveTangent=0.0641, tangentMode=RCTM_User, time=0.3, value=4}
        <at 0.75 0 RCIM_Cubic>
       ]}}})
   &ScaleColor(
    Scale Alpha=&FloatFromCurve(
     FloatCurve=DataInterface() {propertyValues {Curve {keys [{arriveTangent=-2.0238, interpMode=RCIM_Cubic, leaveTangent=-2.0238, tangentMode=RCTM_User, time=0, value=1}, <at 1 0 RCIM_Cubic>]}}}
     Scale Curve=3
    )
    Scale RGB=&Multiply_VectorByFloat(
     Vector=<1 1 1>
     Float=&MakeCustomFloatFromBool(Bool=$User.bIsPainted, True Float=3.5, False Float=&RandomRangeFloat(Minimum=5, Maximum=12))
    )
    ScaleA=true
    ScaleRGB=true
   )
   &ScaleVelocity(Velocity Scale=&VectorFromFloat(Value=&FloatFromCurve(FloatCurve=DataInterface() {propertyValues {Curve {keys [<at 0 0.5 RCIM_Cubic>,<at 1 0 RCIM_Cubic>]}}}))) {enabled=false}
   &JitterPosition(Jitter Amount:NiagaraFloat=&MakeFloatFromVector(Vector=&RandomRangeVector(Minimum=<0 -1.5 0>, Maximum=<0 1.5 0>), Channel=::NewEnumerator1))
   &SolveForcesAndVelocity(Clamp Velocity=false, Limit Acceleration=false)
   &DynamicMaterialParameters(
    Index 0 Param 1=&MakeCustomFloatFromBool(Bool=$User.bIsPainted, True Float=1, False Float=0)
    Module.Index0Param0Write=true
    Module.Index0Param1Write=true
    Module.Index0Param2Write=true
    Module.Index0Param3Write=true
   )
  ]
  rendererClasses [&NiagaraSpriteRendererProperties]
  renderers [{rendererClass=&NiagaraSpriteRendererProperties}]
 }
}
summary {
 systemName=NS_Boost_Digital
 emitters [
  {emitterName=NE_VC_Boost_Sprite_Parent, rendererClasses [&NiagaraSpriteRendererProperties]}
  {emitterName=NE_Boost_Ribbon, rendererClasses [&NiagaraRibbonRendererProperties]}
 ]
 userVariables [
  User.bIsFrontendPreview {type=NiagaraBool}
  User.bIsPainted {type=NiagaraBool}
  User.ColorVariantID {type=NiagaraInt32}
  User.DTCarOffset {type=Vector3f, description="", defaultValue {struct=Vector3f, value=<0 0 0>}}
  User.GlowOverride {type=NiagaraFloat, defaultValue=4, description=""}
  User.PaintedColor {type=LinearColor, description="", defaultValue {struct=LinearColor, value=<rgba 1 1 1 1>}}
  User.PaintedScalar {type=NiagaraFloat, defaultValue=1, description=""}
  User.PortCount {type=NiagaraInt32, defaultValue=1}
 ]
}
